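# Niagara swing trail for the Cloaked Iron pickaxe, built from four anim-trail ribbon emitters.
# SoftTrail and TipTrail draw faint dark and white ribbons, while TrailMagenta and TrailGem add magenta and red glows from weapon bones or sockets.
# All fade quickly by curve and are tinted by a System.TOD color, with time-of-day alpha scaling on both glows.
# TrailGem also tapers and computes a wider gold color/width pair alongside its two ribbon renderers.
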NiagaraSystem /BRCosmetics/Weapons/FORT_Melee/Pickaxe_CloakedIron/FX/NS_Pickaxe_CloakedIron_Trail
dependencies {
 usedDataInterfaces [
  &NiagaraDataInterfaceCurve
  &NiagaraDataInterfaceSkeletalMesh
  &NiagaraDataInterfaceSocketReader
  &NiagaraDataInterfaceColorCurve
 ]
 usedDynamicInputs [&ColorFromCurve]
 usedModules [
  &self:SystemSpawnScript.SystemScriptSource.SystemScriptGraph.NiagaraNodeAssignment_0.SetVariables_#1
  &self:SystemSpawnScript.SystemScriptSource.SystemScriptGraph.NiagaraNodeAssignment_1.SetVariables_#2
  &SystemState
  &self:SystemSpawnScript.SystemScriptSource.SystemScriptGraph.NiagaraNodeAssignment_2.SetVariables_#3
 ]
}
emitters {
 SoftTrail {
  emitterUpdateScript [
   &EmitterState()
   &FN_SpawnAnimTrail(AnimTrailWidthMode=::FromFirst)
  ]
  particleSpawnScript [
   &FN_InitializeAnimTrailBoneIndecies()
   &FN_InitializeAnimTrail(AnimTrailWidth=3, Color=<rgba 0.01 0.01 0.01 0.3>, Lifetime=0.2, SkeletalMesh=$System.Skeletal Mesh) {module=FN_InitializeAnimTrail001}
  ]
  particleUpdateScript [
   &FN_UpdateAnimTrail(SkeletalMesh=$System.Skeletal Mesh) {module=FN_UpdateAnimTrail001}
   &ScaleColor(
    Scale Alpha=&FloatFromCurve(FloatCurve=DataInterface() {propertyValues {Curve {keys [
         {arriveTangent=-0.3998, interpMode=RCIM_Cubic, leaveTangent=-0.3998, tangentMode=RCTM_User, time=0, value=1}
         {arriveTangent=-2.0275, interpMode=RCIM_Cubic, leaveTangent=-2.0275, tangentMode=RCTM_User, time=1, value=0}
        ]}}})
    Scale RGB=&MakeVectorFromLinearColorRGB(Color=$System.TOD)
    ScaleA=true
    ScaleRGB=true
   )
  ]
  rendererClasses [&NiagaraRibbonRendererProperties]
  renderers [{rendererClass=&NiagaraRibbonRendererProperties}]
 }
 TipTrail {
  emitterUpdateScript [
   &EmitterState()
   &FN_SpawnAnimTrail(AnimTrailWidthMode=::FromSecond)
  ]
  particleSpawnScript [
   &FN_InitializeAnimTrailBoneIndecies()
   &FN_InitializeAnimTrail(Color=<rgba 1 1 1 0.6>, Lifetime=0.125, SkeletalMesh=$System.Skeletal Mesh) {module=FN_InitializeAnimTrail001}
  ]
  particleUpdateScript [
   &FN_UpdateAnimTrail(SkeletalMesh=$System.Skeletal Mesh) {module=FN_UpdateAnimTrail001}
   &ScaleColor(
    Scale Alpha=&FloatFromCurve(
     FloatCurve=DataInterface() {propertyValues {Curve {keys [
         {arriveTangent=-0.3998, interpMode=RCIM_Cubic, leaveTangent=-0.3998, tangentMode=RCTM_User, time=0, value=1}
         {arriveTangent=-2.0275, interpMode=RCIM_Cubic, leaveTangent=-2.0275, tangentMode=RCTM_User, time=1, value=0}
        ]}}}
     Scale Curve=1.5
    )
    Scale RGB=&MakeVectorFromLinearColorRGB(Color=$System.TOD)
    ScaleA=true
    ScaleRGB=true
   )
  ]
  rendererClasses [&NiagaraRibbonRendererProperties]
  renderers [{rendererClass=&NiagaraRibbonRendererProperties}]
 }
 TrailGem {
  emitterUpdateScript [
   &EmitterState()
   &FN_SpawnAnimTrail()
  ]
  particleSpawnScript [
   &FN_InitializeAnimTrailBoneIndecies(First Bone or Socket Index=2, Second Bone or Socket Index=2) {module=FN_InitializeAnimTrailBoneIndecies001}
   &FN_InitializeAnimTrail(AnimTrailWidth=13, Color=<rgba 2 0.1 0.1 0.5>, Lifetime=0.125) {module=FN_InitializeAnimTrail001}
   &SystemLocation(MaskSpawnGroup=false)
  ]
  particleUpdateScript [
   &ScaleColor(
    Scale Alpha=&Multiply_Float(
     A=&FloatFromCurve(FloatCurve=DataInterface() {propertyValues {Curve {keys [
          {arriveTangent=-0.3998, interpMode=RCIM_Cubic, leaveTangent=-0.3998, tangentMode=RCTM_User, time=0, value=1}
          {arriveTangent=-2.0275, interpMode=RCIM_Cubic, leaveTangent=-2.0275, tangentMode=RCTM_User, time=1, value=0}
         ]}}})
     B=&FNTimeOfDayScalar(Night Scale=1, Morning Scale=0.65, Evening Scale=0.5)
    )
    Scale RGB=&MakeVectorFromLinearColorRGB(Color=$System.TOD)
    ScaleA=true
    ScaleRGB=true
   ) {module=ScaleColor002}
   &FN_UpdateAnimTrail()
   &ScaleRibbonWidth(Ribbon Width Scale=&FloatFromCurve(FloatCurve=DataInterface() {propertyValues {Curve {
        defaultValue=1
        keys [
         {arriveTangent=-0.6704, interpMode=RCIM_Cubic, leaveTangent=-0.6704, tangentMode=RCTM_User, time=0, value=1}
         {interpMode=RCIM_Cubic, tangentMode=RCTM_User, time=1, value=0}
        ]
       }}}))
   &DynamicMaterialParameters(
    Index 1 Param 2=0.5
    Module.Index1Float0Write=true
    Module.Index1Float1Write=true
    Module.Index1Float2Write=true
    Module.Index1Float3Write=true
    Write Parameter Index 0=::NewEnumerator0
    Write Parameter Index 1=::NewEnumerator1
   )
   /SetVariables(
    Particles.GoldColor:LinearColor=&MakeLinearColorFromVectorAndFloat(
     "Vector (RGB)"=&Multiply_Vector(
      A=&MakeVectorFromLinearColorRGB(Color=$System.TOD)
      B=&MakeVectorFromLinearColorRGB(Color=<rgba 2 0.9827 0.4 0.3>)
     )
     "Float (Alpha)"=&Multiply_Float(
      A=&FloatFromCurve(FloatCurve=DataInterface() {propertyValues {Curve {keys [
           {arriveTangent=-0.3998, interpMode=RCIM_Cubic, leaveTangent=-0.3998, tangentMode=RCTM_User, time=0, value=1}
           {arriveTangent=-2.0275, interpMode=RCIM_Cubic, leaveTangent=-2.0275, tangentMode=RCTM_User, time=1, value=0}
          ]}}})
      B=&FNTimeOfDayScalar(Night Scale=1, Morning Scale=0.65, Evening Scale=0.5)
     )
    )
    Particles.GoldWidth:NiagaraFloat=&Multiply_Float(A=$Particles.RibbonWidth, B=1.8)
   ) {isSetParameters=true}
  ]
  rendererClasses [&NiagaraRibbonRendererProperties]
  renderers [
   {rendererClass=&NiagaraRibbonRendererProperties}
   {rendererClass=&NiagaraRibbonRendererProperties, rendererIndex=1}
  ]
 }
 TrailMagenta {
  emitterUpdateScript [
   &EmitterState()
   &FN_SpawnAnimTrail()
  ]
  particleSpawnScript [
   &FN_InitializeAnimTrailBoneIndecies() {module=FN_InitializeAnimTrailBoneIndecies001}
   &FN_InitializeAnimTrail(AnimTrailWidth=2, Color=<rgba 3 0.18 1.825 0.4>, Lifetime=0.125) {module=FN_InitializeAnimTrail001}
   &SystemLocation(MaskSpawnGroup=false)
  ]
  particleUpdateScript [
   &ScaleColor(
    Scale Alpha=&Multiply_Float(
     A=&FloatFromCurve(FloatCurve=DataInterface() {propertyValues {Curve {keys [
          {arriveTangent=-0.3998, interpMode=RCIM_Cubic, leaveTangent=-0.3998, tangentMode=RCTM_User, time=0, value=1}
          {arriveTangent=-2.0275, interpMode=RCIM_Cubic, leaveTangent=-2.0275, tangentMode=RCTM_User, time=1, value=0}
         ]}}})
     B=&FNTimeOfDayScalar(Night Scale=1, Morning Scale=0.65, Evening Scale=0.5)
    )
    Scale RGB=&MakeVectorFromLinearColorRGB(Color=$System.TOD)
    ScaleA=true
    ScaleRGB=true
   ) {module=ScaleColor002}
   &FN_UpdateAnimTrail()
   &DynamicMaterialParameters(
    Index 1 Param 2=2
    Module.Index1Float0Write=true
    Module.Index1Float1Write=true
    Module.Index1Float2Write=true
    Module.Index1Float3Write=true
    Write Parameter Index 0=::NewEnumerator0
    Write Parameter Index 1=::NewEnumerator1
   )
  ]
  rendererClasses [&NiagaraRibbonRendererProperties]
  renderers [{rendererClass=&NiagaraRibbonRendererProperties}]
 }
}
summary {
 systemName=NS_Pickaxe_CloakedIron_Trail
 emitters [
  {emitterName=SoftTrail, rendererClasses [&NiagaraRibbonRendererProperties]}
  {emitterName=TipTrail, rendererClasses [&NiagaraRibbonRendererProperties]}
  {emitterName=TrailMagenta, rendererClasses [&NiagaraRibbonRendererProperties]}
  {emitterName=TrailGem, rendererClasses [&NiagaraRibbonRendererProperties]}
 ]
 userVariables [
  User.bIsFrontEnd {type=NiagaraBool}
  User.bIsFrontEndPreview {type=NiagaraBool}
  User.PhysicalSurface {type=NiagaraInt32}
  User.RibbonIDOffset {type=NiagaraInt32}
  User.SwingSectionIndex {type=NiagaraInt32}
 ]
}
